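# Translucent unlit Niagara mesh-particle material for the Medicrow skydiving contrail, with a starry galaxy look.
# Emissive combines noise-driven curve-atlas color, an optional start/end gradient and camera-vector triplanar stars, tinted by particle color and time of day.
# Opacity uses a panned, noise-distorted trail mask with dissolve, optional tail and detail masks, and camera and depth fades.
# World position offset adds sine waviness to SplineThicken, with width, wave, panning and dissolve driven by dynamic parameters.
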Material /BRCosmetics/Athena/Effects/Particles/Skydiving/NS_Contrail_Medicrow/M_Contrail_Medicrow
graph {
 properties {
  bUsedWithNiagaraMeshParticles=true
  blendMode=BLEND_Translucent
  pixelDepthOffsetMode=PDOM_Legacy
  shadingModel=MSM_Unlit
  twoSided=true
  usedShadingModels=Unlit
 }
 nodes [
  n1 = TextureSample(UVs=Panner(Coordinate=TextureCoordinate()) {speedX=-0.1, speedY=0.5}) {samplerType=SAMPLERTYPE_Grayscale, texture=&LowResBlurredNoise}
  n2 = Add(A=Add(A=TextureCoordinate() {uTiling=2}, B=Multiply(A=@n1.R) {constB=0.2}), B=Constant2Vector())
  n3 = Multiply(
   A=Constant3Vector() {constant=<rgba 0 1.6029 6 0>}
   B=TextureSample(UVs=Panner(Coordinate=@n2) {speedX=-0.7}) {samplerType=SAMPLERTYPE_LinearGrayscale, texture=/Game/Global/Textures/Default/Caustics/T_Nosie_Caustics_Grayscale}.RGB
  )
  n4 = TextureSample() {texture=&T_Inky_Smoke_Tile}
  n5 = Constant() {r=1}
  n6 = CurveAtlasRowParameter(CurveTime=ComponentMask(None=TextureCoordinate()) {r=true}) {
   atlas=/Game/Athena/LinearColorCurves/Curve_Atlas_Contrail_Inferno
   curve=/Game/Athena/LinearColorCurves/Contrail_Inferno_ColorCurve
   parameterName=Color Curve
  }
  n7 = TextureCoordinate() {uTiling=0.5, vTiling=0.5}
  Add(
   A=Add(A=Multiply(A=@n3, B=@n4.R), B=Multiply(A=@n5, B=@n6))
   B=Desaturation(
    None=TextureSample(UVs=@n7) {texture=/Game/Gadgets/Assets/VinderTech_GliderChute/Glider_Cube_Queen_Female/FX/T_CubeQueen_Glider_Stair}.RGB
    Fraction=Constant() {r=-1}
   )
  ) {constB=0}
  n8 = Constant() {r=40}
  n9 = DynamicParameter() {parameterIndex=1, defaultValue=<rgba 1 1 0 1>, paramNames [Width,none,none,none]}
  n10 = Constant() {r=80}
  n11 = Add(A=TextureCoordinate(), B=AppendVector(A=Constant(), B=Constant()))
  n12 = Time()
  n13 = Sine(None=Add(A=ComponentMask(None=@n11) {r=true}, B=Multiply(A=@n12) {constB=-0.5})) {period=0.6}
  n14 = Constant3Vector() {constant=<rgba 0 1 0 0>}
  n15 = Multiply(
   A=Transform(None=Multiply(A=@n13, B=@n14)) {transformSourceType=TRANSFORMSOURCE_Local}
   B=Multiply(
    A=ScalarParameter() {defaultValue=10, parameterName=WPO Intensity}
    B=DynamicParameter() {defaultValue=<rgba 1 1 0 1>, paramNames [WPO, panner, offset, dissolve offset]}.WPO
   )
  )
  n16 = TextureCoordinate()
  n17 = Saturate(None=Multiply(A=ComponentMask(None=@n16) {r=true}) {constB=3})
  Add(
   A=MaterialFunctionCall(WidthBase=Multiply(A=@n8, B=@n9.Width), WidthTip=Multiply(A=@n10, B=@n9.Width), Expand U or V UV Channel=StaticBool()) {materialFunction=&SplineThicken}.WorldPositionOffset
   B=Multiply(A=Add(A=@n15) {constB=0}, B=LinearInterpolate(Alpha=@n17))
  )
  n18 = Add(
   A=TextureCoordinate() {uTiling=0.5, vTiling=0.5}
   B=ComponentMask(None=VectorParameter() {parameterName=Color mask offset, defaultValue=<rgba 0.8 0.4 0 1>}.RGB) {g=true, r=true}
  )
  n19 = CurveAtlasRowParameter(CurveTime=Power(Base=TextureSample(UVs=@n18) {samplerType=SAMPLERTYPE_Grayscale, texture=&LowResBlurredNoise}.R) {constExponent=2.5}) {
   atlas=&Curve_Atlas_Contrail_Inferno
   curve=/Game/Athena/LinearColorCurves/Contrail_Inferno_ColorCurve_2
   defaultValue=1
  }
  n20 = Multiply(A=@n19)
  n21 = TextureObject() {texture=/Game/Gadgets/Assets/VinderTech_GliderChute/Glider_Cube_Queen_Female/FX/T_Cheap_blue_space_color}
  n22 = Constant3Vector() {constant=<rgba 3 3 3 0>}
  Add(
   A=@n20
   B=Multiply(A=MaterialFunctionCall!(CameraVector, Texture=@n21, Tiling=@n22, Offset, AxisFadeContrast) {materialFunction=/Engine/Functions/Engine_MaterialFunctions01/Texturing/TriplanarCameraVector}.Result) {constB=0.4}
  )
  n23 = ParticleColor()
  n24 = MaterialFunctionCall(Distance=ScalarParameter() {defaultValue=150, parameterName=Near camera fade}) {materialFunction=&Near_Camera_Fade}
  n25 = Add(
   A=TextureCoordinate()
   B=AppendVector(A=DynamicParameter() {defaultValue=<rgba 1 1 0 1>, paramNames [WPO, panner, offset, dissolve offset]}.offset, B=Constant())
  )
  n26 = TextureCoordinate()
  n27 = Multiply(A=TextureSample(UVs=Panner(Coordinate=@n26) {speedX=-0.3}) {samplerType=SAMPLERTYPE_Grayscale, texture=&LowResBlurredNoise}.R) {constB=0.2}
  n28 = Constant2Vector() {r=-1}
  n29 = DynamicParameter() {defaultValue=<rgba 1 1 0 1>, paramNames [WPO, panner, offset, dissolve offset]}
  n30 = TextureSampleParameter2D(UVs=Panner(Coordinate=Add(A=@n25, B=@n27), Speed=Multiply(A=@n28, B=@n29.panner)) {speedX=-1}) {parameterName=Mask Tex, texture=/BRCosmetics/Athena/Effects/Particles/Skydiving/NS_Contrail_Inferno/T_Contrail_Inferno_Trail}
  n31 = ScalarParameter() {defaultValue=-1, parameterName=Dissolve_offsetA}
  n32 = ScalarParameter() {defaultValue=1.5, parameterName=Dissolve_offsetB}
  n33 = DynamicParameter() {defaultValue=<rgba 1 1 0 1>, paramNames [WPO, panner, offset, dissolve offset]}
  n34 = Constant()
  n35 = Add(A=Add(A=TextureCoordinate(), B=AppendVector(A=@n33.dissolve offset, B=@n34)), B=Constant2Vector() {g=0.2})
  n36 = Add(A=ComponentMask(None=TextureCoordinate()) {r=true}) {constB=-0.5}
  n37 = Power(
   Base=TextureSample(UVs=Panner(Coordinate=@n35) {speedX=-1}) {samplerType=SAMPLERTYPE_Grayscale, texture=&LowResBlurredNoise}.R
   Exp=LinearInterpolate(A=Constant() {r=0.1}, B=Constant() {r=2}, Alpha=Saturate(None=@n36))
  )
  n38 = TextureCoordinate()
  n39 = Power(Base=OneMinus(None=ComponentMask(None=@n38) {r=true})) {constExponent=7}
  n40 = Multiply(
   A=Multiply(
    A=StaticComponentMaskParameter(None=@n30.RGB) {defaultR=true, parameterName=RGB mask}
    B=LinearInterpolate(A=@n31, B=@n32, Alpha=@n37) {constA=-1, constB=1.5}
   )
   B=LinearInterpolate(B=Saturate(None=@n39), Alpha=ScalarParameter() {parameterName=Use tail mask}) {constA=1}
  )
  n41 = Panner(Coordinate=Add(A=TextureCoordinate() {uTiling=0.5, vTiling=0.5}, B=Constant2Vector() {r=0.2}))
  n42 = LinearInterpolate(
   B=Multiply(A=TextureSample(UVs=@n41) {texture=&T_Inky_Smoke_Tile}.R) {constB=5}
   Alpha=ScalarParameter() {parameterName=Use Detail apha mask}
  ) {constA=1}
  n43 = Power(Base=SphereMask(A=TextureCoordinate(), B=Constant2Vector() {g=0.5, r=0.6}) {attenuationRadius=0.6, hardnessPercent=0})
  n44 = Multiply(
   A=Multiply(A=Multiply(A=@n40, B=@n42), B=ScalarParameter() {defaultValue=1, parameterName=Main Alpha})
   B=Saturate(None=Multiply(A=@n43) {constB=50})
  )
  DepthFade(Opacity=Multiply(A=@n23.A, B=Multiply(A=@n44, B=@n24.Result)), FadeDistance=ScalarParameter() {defaultValue=50, parameterName=DepthFade})
  n45 = Constant() {r=1}
  n46 = VectorParameter() {parameterName=Color Night, defaultValue=<rgba 2 0.6849 1.5726 0>}
  n47 = TextureObjectParameter() {
   group=Galaxy
   parameterName=Stars
   samplerType=SAMPLERTYPE_Masks
   texture=/Game/Characters/Player/Male/Medium/Bodies/M_Med_Soldier_04/Skins/Galaxy/Textures/Celestial_T_Stars
  }
  n48 = ScalarParameter() {defaultValue=4, group=Galaxy, parameterName=Bright Star Tiling}
  n49 = ComponentMask(None=MaterialFunctionCall(Texture=@n47, Tiling=Multiply(A=@n48)) {materialFunction=&TriplanarCameraVector}.Result) {r=true}
  n50 = Time()
  n51 = TextureObject() {samplerType=SAMPLERTYPE_Masks, texture=&Celestial_T_Stars}
  n52 = ComponentMask(None=MaterialFunctionCall!(
    CameraVector
    Rotation Speed=Multiply(A=@n50) {constB=-0.125}
    Texture=@n51
    Tiling=Constant3Vector() {constant=<rgba 8 8 8 0>}
    Offset
    AxisFadeContrast
   ) {materialFunction=/Game/Weapons/WeaponSkins/MaterialFunction/TriplanarCameraVector_Rotated}.Result) {b=true}
  n53 = Add(A=Multiply(A=@n49, B=@n49), B=Multiply(A=@n52, B=@n52))
  n54 = TextureSample(UVs=Panner(Coordinate=TextureCoordinate() {uTiling=2, vTiling=2}) {speedX=0.133, speedY=0.2}) {samplerType=SAMPLERTYPE_Masks, texture=&Celestial_T_Stars}
  n55 = Power(Base=@n54.G) {constExponent=1.7}
  n56 = Multiply(A=@n20)
  n57 = TextureCoordinate() {vTiling=0.5}
  n58 = Constant2Vector() {g=-0.05, r=0.7}
  n59 = Multiply(
   A=VectorParameter() {parameterName=EndColor, defaultValue=<rgba 0.2284 0.0012 1 1>}.RGB
   B=TextureSample(UVs=Add(A=@n57, B=@n58)) {samplerType=SAMPLERTYPE_Grayscale, texture=&LowResBlurredNoise}.R
  )
  n60 = VectorParameter() {parameterName=StartColor, defaultValue=<rgba 1.9576 0.5757 3 1>}
  n61 = TextureCoordinate()
  n62 = LinearInterpolate(
   A=Add(A=@n59, B=@n60.RGB)
   B=Add(A=@n59, B=Multiply(A=@n20, B=@n60.RGB)) {constB=0}
   Alpha=Power(Base=MaterialFunctionCall(UV Channel=@n61) {materialFunction=&LinearGradient}.UGradient) {constExponent=0.25}
  )
  n63 = ScalarParameter() {parameterName=Use Center trail color}
  n64 = Add(A=Multiply(A=Multiply(A=@n53, B=@n55)) {constB=0}, B=Add(A=LinearInterpolate(A=@n56, B=@n62, Alpha=@n63)) {constB=0})
  MaterialFunctionCall(Emissive=Multiply(
    A=MaterialFunctionCall(Color_Morning=@n45, Color_Day=@n45, Color_Evening=@n45, Color_Night=@n46.RGB) {materialFunction=&Color_Time_of_Day}.Result
    B=Multiply(A=@n64, B=@n23.RGB)
   )) {materialFunction=&MF_FX_EmissiveAdjustments}
  ParticleColor()
 ]
}
