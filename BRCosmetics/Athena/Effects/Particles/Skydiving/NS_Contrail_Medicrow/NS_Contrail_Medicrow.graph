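# Niagara skydiving contrail for the Medicrow cosmetic, drawing purple flame trails from four skeletal mesh bones or sockets.
# Trail_Purple holds four persistent mesh and sprite particles, while two emitters add rising, curve-scaled flame sprites.
# Partition index selects attachment and per-point sizing; FN_Contrail_Setup's dive value and owner velocity drive stretch and dynamic material parameters.
# Front-end modifiers adjust the look outside gameplay; the Frontend_Flames emitter is disabled.
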NiagaraSystem /BRCosmetics/Athena/Effects/Particles/Skydiving/NS_Contrail_Medicrow/NS_Contrail_Medicrow
dependencies {
 usedDataInterfaces [
  &NiagaraDataInterfaceArrayFloat3
  &NiagaraDataInterfaceCurve
  &NiagaraDataInterfaceArrayFloat
  &NiagaraDataInterfaceColorCurve
  &NiagaraDataInterfaceSkeletalMesh
 ]
 usedDynamicInputs []
 usedModules [
  &self:SystemSpawnScript.SystemScriptSource.SystemScriptGraph.NiagaraNodeAssignment_0.SetVariables_#1
  &SystemState
  /Game/Effects/Niagara/Modules/Contrails/FN_Contrail_Setup
  &FN_Integrator
 ]
}
emitters {
 FlameSprites {
  emitterUpdateScript [
   &EmitterState()
   &SpawnRate(SpawnRate=80, Use Spawn Probability=false)
  ]
  particleSpawnScript [
   &PartitionParticles(Number Of Partitions=4)
   &InitializeParticle(
    Color=<rgba 0.4 0.4 0.4 1>
    Lifetime Max=1
    Lifetime Min=0.8
    Lifetime Mode=::NewEnumerator1
    Sprite Rotation Mode=::NewEnumerator2
    Sprite Size Mode=::NewEnumerator1
    Uniform Sprite Size Max=100
    Uniform Sprite Size Min=50
    UsePositionOffset=false
    Write Material Random=true
   ) {module=InitializeParticle001}
   &SphereLocation(
    MaskSpawnGroup=false
    Override Seed=false
    Sphere Radius=15
    Surface Only=false
    UseNonUniformScale=false
    UseOffset=false
    UseSphereOrientation=false
   )
   &AddVelocity(
    Cone Angle=30
    Cone Axis=&Lerp_Vector(A=<0 0 1>, B=<0 0 0.5>, Alpha=&FN_FrontEndFloatModifier(Gameplay=0, Front End=1))
    Module.UseConeFalloff=false
    Module.UseVelDistribution=false
    Override Seed=false
    Velocity Mode=::NewEnumerator2
    Velocity Speed=100
   )
  ]
  particleUpdateScript [
   &SkeletalMeshLocation(
    Bone / Socket Index=$Particles.PartitionParticles.Partition
    Bone Sampling Mode=::NewEnumerator4
    MaskSpawnGroup=false
    Orientation Sampling=::NewEnumerator1
    Position Sampling=::NewEnumerator1
    Skeletal Mesh=$System.Skeletal Mesh
   ) {module=SkeletalMeshLocation001}
   /SetVariables(Particles.SubImageIndex=&FloatFromCurve(FloatCurve=DataInterface() {propertyValues {Curve {keys [<at 0 5>,<at 1 15>]}}})) {isSetParameters=true}
   &ParticleState()
   &Drag(Drag=2, Use Linear Drag=true, Use Rotational Drag=false)
   &AccelerationForce(Acceleration=<0 0 350>)
   &SolveForcesAndVelocity(Clamp Velocity=false, Limit Acceleration=false)
   &ScaleSpriteSize(Scale Factor=&Vector2DFromFloat(Value=&FloatFromCurve(
      FloatCurve=DataInterface() {propertyValues {Curve {keys [
          {interpMode=RCIM_Cubic, leaveTangent=5.3269, tangentMode=RCTM_Break, time=0, value=0}
          {arriveTangent=0.6, interpMode=RCIM_Cubic, leaveTangent=0.6, time=0.2, value=0.54}
          {arriveTangent=0.0611, interpMode=RCIM_Cubic, tangentMode=RCTM_Break, time=1, value=0.6}
         ]}}}
      Scale Curve=&SelectFloatFromArray(Direct Array Index=$Particles.PartitionParticles.Partition, Float Selection Array=DataInterface() {propertyValues {FloatData [1,1,0.6,0.6]}})
     )))
   &ScaleColor(
    Scale Alpha=&Multiply_Float(
     A=&FloatFromCurve(FloatCurve=DataInterface() {propertyValues {Curve {keys [
          {interpMode=RCIM_Cubic, tangentMode=RCTM_User, time=0, value=0}
          {interpMode=RCIM_Cubic, tangentMode=RCTM_User, time=0.5, value=1}
          {tangentMode=RCTM_User, time=1, value=0}
         ]}}})
     B=&RandomRangeFloat(Minimum=1, Maximum=2)
    )
    Scale RGB=&Multiply_Vector(
     A=<2 2 2>
     B=&MakeVectorFromLinearColorRGB(Color=&ColorFromCurve(ColorCurve=DataInterface() {propertyValues {
         AlphaCurve {keys []}
         BlueCurve {keys [<at 0 1 RCIM_Cubic>,<at 0.428 0.0412>]}
         GreenCurve {keys [<at 0 0.3332 RCIM_Cubic>,<at 0.428 0>]}
         RedCurve {keys [<at 0 0.8231 RCIM_Cubic>,<at 0.428 0.0124>]}
        }}))
    )
    ScaleA=true
    ScaleRGB=true
   ) {module=ScaleColor001}
   &CameraOffset(Camera Offset Amount=&RandomRangeFloat(
     Minimum=&SelectFloatFromArray(Direct Array Index=$Particles.PartitionParticles.Partition, Float Selection Array=DataInterface() {propertyValues {FloatData [25,25,5,5]}})
     Maximum=1
    ))
  ]
  rendererClasses [&NiagaraSpriteRendererProperties]
  renderers [{rendererClass=&NiagaraSpriteRendererProperties}]
 }
 FlameSprites001 {
  emitterUpdateScript [
   &EmitterState()
   &SpawnRate(SpawnRate=40, Use Spawn Probability=false)
  ]
  particleSpawnScript [
   &PartitionParticles(Number Of Partitions=4)
   /SetVariables(Particles.Random size:NiagaraFloat=&RandomRangeFloat(Minimum=0.2, Maximum=0.25)) {isSetParameters=true}
   &InitializeParticle(
    Color=<rgba 0.4 0.4 0.4 1>
    Lifetime Max=1
    Lifetime Min=0.8
    Lifetime Mode=::NewEnumerator1
    Sprite Size Mode=::NewEnumerator3
    Uniform Sprite Size=&Multiply_Float(A=100, B=$Particles.Random size)
    UsePositionOffset=false
    Write Material Random=true
   ) {module=InitializeParticle001}
   &SphereLocation(
    MaskSpawnGroup=false
    Override Seed=false
    Sphere Radius=15
    Surface Only=false
    UseNonUniformScale=false
    UseOffset=false
    UseSphereOrientation=false
   )
   &AddVelocity(
    Cone Angle=30
    Cone Axis=<0 0 1>
    Module.UseConeFalloff=false
    Module.UseVelDistribution=false
    Override Seed=false
    Velocity Mode=::NewEnumerator2
    Velocity Speed=100
   )
   /SetVariables() {isSetParameters=true}
   /SetVariables(Particles.DY3:Vector4f=&MakeVector4(X=3, Y=1, Z=1, W=1)) {isSetParameters=true}
  ]
  particleUpdateScript [
   /SetVariables(Particles.Size Scale:NiagaraFloat=&FloatFromCurve(FloatCurve=DataInterface() {propertyValues {Curve {keys [<at 0 0>,<at 0.1 0.5>,<at 1 0>]}}})) {isSetParameters=true}
   &SkeletalMeshLocation(
    Bone / Socket Index=$Particles.PartitionParticles.Partition
    Bone Sampling Mode=::NewEnumerator4
    MaskSpawnGroup=false
    Orientation Sampling=::NewEnumerator1
    Position Sampling=::NewEnumerator1
    Skeletal Mesh=$System.Skeletal Mesh
   ) {module=SkeletalMeshLocation001}
   &SubUVAnimation(End Frame=11, Start Frame=8, SubUV Animation Mode:ENiagara_SubUVLookupMode=::NewEnumerator2)
   &ParticleState()
   &Drag(
    Drag=&SelectFloatFromArray(
     Direct Array Index=$Particles.PartitionParticles.Partition
     Float Selection Array=DataInterface() {propertyValues {FloatData [1.25,1.25,2,2]}}
    )
    Use Linear Drag=true
    Use Rotational Drag=false
   )
   &AccelerationForce(Acceleration=&Lerp_Vector(
     A=<0 0 400>
     B=<0 0 600>
     Alpha=&FN_FrontEndFloatModifier(Gameplay=$System.FN_Contrail_Setup.Dive Value, Front End=0)
    ))
   &SolveForcesAndVelocity(Clamp Velocity=false, Limit Acceleration=false)
   &ScaleSpriteSize(Scale Factor=&Vector2DFromFloat(Value=&FloatFromCurve(
      FloatCurve=DataInterface() {propertyValues {Curve {keys [
          {interpMode=RCIM_Cubic, leaveTangent=5.3269, tangentMode=RCTM_Break, time=0, value=0}
          {arriveTangent=0.6, interpMode=RCIM_Cubic, leaveTangent=0.6, time=0.2, value=0.54}
          {arriveTangent=0.0611, interpMode=RCIM_Cubic, tangentMode=RCTM_Break, time=1, value=0.6}
         ]}}}
      Scale Curve=&SelectFloatFromArray(
       Direct Array Index=$Particles.PartitionParticles.Partition
       Float Selection Array=DataInterface() {propertyValues {FloatData [0.5,0.5,0.3,0.3]}}
      )
     )))
   &ScaleColor(
    Scale Alpha=&Multiply_Float(
     A=&FloatFromCurve(FloatCurve=DataInterface() {propertyValues {Curve {keys [
          {interpMode=RCIM_Cubic, tangentMode=RCTM_User, time=0, value=0}
          {interpMode=RCIM_Cubic, tangentMode=RCTM_User, time=0.5, value=1}
          {tangentMode=RCTM_User, time=1, value=0}
         ]}}})
     B=&RandomRangeFloat(Minimum=1, Maximum=2)
    )
    Scale RGB=&Multiply_Vector(
     A=<2 2 2>
     B=&MakeVectorFromLinearColorRGB(Color=&ColorFromCurve(ColorCurve=DataInterface() {propertyValues {
         AlphaCurve {keys []}
         BlueCurve {keys [<at 0 1 RCIM_Cubic>]}
         GreenCurve {keys [<at 0 0.3332 RCIM_Cubic>]}
         RedCurve {keys [<at 0 0.8231 RCIM_Cubic>]}
        }}))
    )
    ScaleA=true
    ScaleRGB=true
   ) {module=ScaleColor001}
   &CameraOffset(Camera Offset Amount=&RandomRangeFloat(
     Minimum=&SelectFloatFromArray(Direct Array Index=$Particles.PartitionParticles.Partition, Float Selection Array=DataInterface() {propertyValues {FloatData [25,25,5,5]}})
     Maximum=1
    ))
   &FN_Sprite_Attributes(
    Color=&MakeLinearColorFromVectorAndFloat(
     "Vector (RGB)"=&MakeVectorFromLinearColorRGB(Color=&ColorFromCurve(ColorCurve=DataInterface() {propertyValues {
         AlphaCurve {keys []}
         BlueCurve {keys [<at 0 1 RCIM_Cubic>]}
         GreenCurve {keys [<at 0 0.3332 RCIM_Cubic>]}
         RedCurve {keys [<at 0 0.8231 RCIM_Cubic>]}
        }}))
     "Float (Alpha)"=2
    )
    Create Color Attribute=true
    Create SpriteSize Attribute=true
    SpriteSize=&Vector2DFromFloat(Value=&Multiply_Float(A=50, B=&Multiply_Float(A=$Particles.Random size, B=$Particles.Size Scale)))
    Write to Module Namespace=true
   )
   &FN_Sprite_Attributes(
    Color=&MakeLinearColorFromVectorAndFloat(
     "Vector (RGB)"=&MakeVectorFromLinearColorRGB(Color=<rgba 0.7495 0.2303 1.5 0.5>)
     "Float (Alpha)"=&Multiply_Float(
      A=&FN_FrontEndFloatModifier(Gameplay=0.2, Front End=0.1)
      B=&FloatFromCurve(
       FloatCurve=DataInterface() {propertyValues {Curve {keys [{interpMode=RCIM_Cubic, tangentMode=RCTM_User, time=0, value=1}, {interpMode=RCIM_Cubic, tangentMode=RCTM_User, time=1, value=1}]}}}
       Scale Curve=0.5
      )
     )
    )
    Create Color Attribute=true
    Create SpriteSize Attribute=true
    SpriteSize=&Vector2DFromFloat(Value=&Multiply_Float(A=200, B=&Multiply_Float(A=$Particles.Random size, B=$Particles.Size Scale)))
    Write to Module Namespace=true
   ) {module=FN_Sprite_Attributes001}
  ]
  rendererClasses [&NiagaraSpriteRendererProperties]
  renderers [
   {rendererClass=&NiagaraSpriteRendererProperties}
   {rendererClass=&NiagaraSpriteRendererProperties, rendererIndex=1}
  ]
 }
 Frontend_Flames {
  bEnabled=false
  emitterUpdateScript [
   &EmitterState(Life Cycle Mode=::NewEnumerator1, Loop Behavior=::NewEnumerator1, Loop Duration Mode=::NewEnumerator1, UseLoopDelay=false)
   &SpawnBurst_Instantaneous(
    Spawn Count=4
    Spawn Probability=&Lerp_Float(A=0, B=1, Alpha=&FN_FrontEndFloatModifier(Gameplay=0, Front End=1))
    Use Spawn Probability=true
    UseLoopCountLimit=false
   )
  ]
  particleSpawnScript [
   &PartitionParticles(Number Of Partitions=4)
   &InitializeParticle(
    Color=<rgba 1.3627 0.6246 1.5 1>
    Mesh Scale=<1.2 1 1>
    Mesh Scale Mode=::NewEnumerator4
    Sprite Size Mode=::NewEnumerator3
    Uniform Sprite Size=65
    UsePositionOffset=false
    Write Material Random=true
   )
   &InitialMeshOrientation(Mesh Orientation Mode=::NewEnumerator2, Rotation Coordinate Space=::NewEnumerator2, Use Rotation Vector=true) {enabled=false}
   &InitialMeshOrientation(
    Mesh Orientation Mode=::NewEnumerator2
    Rotation=&Lerp_Vector(A=<0 -0.25 0>, B=<0.15 -0.25 0>, Alpha=$User.ParaGlideLeanAlpha)
    Rotation Coordinate Space=::NewEnumerator2
    Use Rotation Vector=true
   ) {module=InitialMeshOrientation001}
   /SetVariables() {isSetParameters=true, module=SetVariables_#2}
  ]
  particleUpdateScript [
   &SkeletalMeshLocation(
    Bone / Socket Index=$Particles.PartitionParticles.Partition
    Bone Sampling Mode=::NewEnumerator4
    MaskSpawnGroup=false
    Orientation Sampling=::NewEnumerator1
    Sampled Position Offset=&SelectVectorFromArray(
     Direct Array Index=$Particles.PartitionParticles.Partition
     Vector Selection Array=DataInterface() {propertyValues {FloatData [<0 0 0>, <0 0 0>]}}
    )
    Skeletal Mesh=$System.Skeletal Mesh
   )
   /SetVariables(
    Particles.Sprite_CameraOffset=25
    Particles.Sprite_Pos:NiagaraPosition=&AddVectorToPosition(Position=$Output.SkeletalMeshLocation.SampledPositionWithoutOffset, Vector=<0 0 0>)
   ) {isSetParameters=true, module=SetVariables_#3}
   &ParticleState(Kill Particles When Lifetime Has Elapsed=false, Let Infinitely Lived Particles Die When Emitter Deactivates=true)
   &ScaleMeshSize(Scale Factor=&SelectVectorFromArray(
     Direct Array Index=$Particles.PartitionParticles.Partition
     Vector Selection Array=DataInterface() {propertyValues {FloatData [<0.8 1 1>, <0.8 1 1>, <0.5 0.5 0.5>, <0.5 0.5 0.5>]}}
    )) {module=ScaleMeshSize001}
   &ScaleMeshSize(Scale Factor=&MakeVector(
     X=&Lerp_Float(
      A=&FN_FrontEndFloatModifier(Gameplay=1, Front End=0.7)
      B=&FN_FrontEndFloatModifier(Gameplay=1.4, Front End=0.8)
      Alpha=&ClampFloat(Float=$System.FN_Contrail_Setup.Dive Value)
     )
     Y=1
     Z=1
    ))
   &DynamicMaterialParameters(
    Index 0 Param 1=&FloatFromCurve(
     FloatCurve=DataInterface() {propertyValues {Curve {keys [<at 0 1>,<at 1 0>]}}}
     CurveIndex=&Divide_Float(A=&VectorLength(Vector=$Engine.Owner.Velocity), B=9000)
    )
    Index 0 Param 2=&MakeCustomFloatFromBool(Bool=&SetBoolByFloatComparison(A=$System.FN_Contrail_Setup.Dive Value, B=0.2), True Float=1.4, False Float=1)
    Index 0 Param 3=&SelectFloatFromArray(
     Direct Array Index=$Particles.PartitionParticles.Partition
     Float Selection Array=DataInterface() {propertyValues {FloatData [0,0.2,0.4,0.6]}}
    )
    Index 0 Param 4=&SelectFloatFromArray(
     Direct Array Index=$Particles.PartitionParticles.Partition
     Float Selection Array=DataInterface() {propertyValues {FloatData [0.05,0.25,0.5,0.75]}}
    )
    Index 1 Param 1=&FN_FrontEndFloatModifier(
     Gameplay=&Lerp_Float(
      A=&SelectFloatFromArray(
       Direct Array Index=$Particles.PartitionParticles.Partition
       Float Selection Array=DataInterface() {propertyValues {FloatData [0.75,0.75,0.3,0.3]}}
      )
      B=&SelectFloatFromArray(
       Direct Array Index=$Particles.PartitionParticles.Partition
       Float Selection Array=DataInterface() {propertyValues {FloatData [0.6,0.6,0.2,0.2]}}
      )
      Alpha=&ClampFloat(Float=$System.FN_Contrail_Setup.Dive Value)
     )
     Front End=0.72
    )
    Module.Param0WriteEnabled=true
    Module.Param0WriteEnabled1=true
    Module.Param1WriteEnabled=true
    Module.Param1WriteEnabled1=true
    Module.Param2WriteEnabled=true
    Module.Param2WriteEnabled1=true
    Module.Param3WriteEnabled=true
    Module.Param3WriteEnabled1=true
    Write Parameter Index 1=true
   )
   /SetVariables(
    Particles.Flame_Color:LinearColor=&Multiply_LinearColorByFloat(Linear Color=<rgba 1 1 1 1>, Float=&FN_FrontEndFloatModifier(Gameplay=1.2, Front End=1))
    Particles.Pink_DynamParam1:Vector4f=&MakeVector4(
     X=&FloatFromCurve(
      FloatCurve=DataInterface() {propertyValues {Curve {keys [<at 0 1>,<at 1 0>]}}}
      CurveIndex=&Divide_Float(A=&VectorLength(Vector=$Engine.Owner.Velocity), B=9000)
     )
     Y=&MakeCustomFloatFromBool(Bool=&SetBoolByFloatComparison(A=$System.FN_Contrail_Setup.Dive Value, B=0.2), True Float=1.4, False Float=1)
     Z=&SelectFloatFromArray(
      Direct Array Index=$Particles.PartitionParticles.Partition
      Float Selection Array=DataInterface() {propertyValues {FloatData [0,0.2,0.4,0.6]}}
     )
     W=0
    )
    Particles.Pink_DynamParam2:Vector4f=&MakeVector4(
     X=&FN_FrontEndFloatModifier(
      Gameplay=&Lerp_Float(
       A=&SelectFloatFromArray(
        Direct Array Index=$Particles.PartitionParticles.Partition
        Float Selection Array=DataInterface() {propertyValues {FloatData [0.8,0.8,0.35,0.35]}}
       )
       B=&SelectFloatFromArray(
        Direct Array Index=$Particles.PartitionParticles.Partition
        Float Selection Array=DataInterface() {propertyValues {FloatData [0.7,0.7,0.3,0.3]}}
       )
       Alpha=&ClampFloat(Float=$System.FN_Contrail_Setup.Dive Value)
      )
      Front End=0.7
     )
     Y=0
     Z=0
     W=0
    )
   ) {isSetParameters=true, module=SetVariables_#4}
   &ScaleColor(
    Scale Alpha=&Waveform(Global Amplitude Scale=2, "Waveform [1]"=::NewEnumerator9, "[1] Frequency"=5, "[1] Amplitude Min/Max"=<0 1>)
    Scale RGB=&Multiply_VectorByFloat(Vector=<1 1 1>, Float=&FN_FrontEndFloatModifier(Gameplay=1.2, Front End=1))
    ScaleA=true
    ScaleRGB=true
   )
  ]
  rendererClasses [&NiagaraMeshRendererProperties, &NiagaraSpriteRendererProperties]
  renderers [
   {rendererClass=&NiagaraMeshRendererProperties, rendererIndex=0}
   {rendererClass=&NiagaraMeshRendererProperties, rendererIndex=1}
   {rendererClass=&NiagaraSpriteRendererProperties, rendererIndex=2}
  ]
 }
 Trail_Purple {
  emitterUpdateScript [
   &EmitterState(Life Cycle Mode=::NewEnumerator1, Loop Behavior=::NewEnumerator1, Loop Duration Mode=::NewEnumerator1, UseLoopDelay=false)
   &SpawnBurst_Instantaneous(
    Spawn Count=4
    Spawn Probability=&Lerp_Float(A=0, B=1, Alpha=&FN_FrontEndFloatModifier(Gameplay=1, Front End=1))
    Use Spawn Probability=true
    UseLoopCountLimit=false
   )
  ]
  particleSpawnScript [
   &PartitionParticles(Number Of Partitions=4)
   &InitializeParticle(
    Color=<rgba 1.3627 0.6246 1.5 1>
    Mesh Scale=<1.2 1 1>
    Mesh Scale Mode=::NewEnumerator4
    Sprite Size Mode=::NewEnumerator3
    Uniform Sprite Size=65
    UsePositionOffset=false
    Write Material Random=true
   )
   &InitialMeshOrientation(Mesh Orientation Mode=::NewEnumerator2, Rotation=<0 -0.25 0>, Use Rotation Vector=true)
   &InitialMeshOrientation(
    Mesh Orientation Mode=::NewEnumerator2
    Rotation=&SelectVectorFromArray(
     Direct Array Index=$Particles.PartitionParticles.Partition
     Vector Selection Array=DataInterface() {propertyValues {FloatData [<0 0 0>, <0 0 0.5>, <0 0 0>, <0 0 0.5>]}}
    )
    Use Rotation Vector=true
   ) {module=InitialMeshOrientation001}
   /SetVariables() {isSetParameters=true, module=SetVariables_#2}
  ]
  particleUpdateScript [
   &SkeletalMeshLocation(
    Bone / Socket Index=$Particles.PartitionParticles.Partition
    Bone Sampling Mode=::NewEnumerator4
    MaskSpawnGroup=false
    Orientation Sampling=::NewEnumerator1
    Sampled Position Offset=&SelectVectorFromArray(
     Direct Array Index=$Particles.PartitionParticles.Partition
     Vector Selection Array=DataInterface() {propertyValues {FloatData [<0 0 0>, <0 0 0>]}}
    )
    Skeletal Mesh=$System.Skeletal Mesh
   )
   /SetVariables(
    Particles.Sprite_CameraOffset=25
    Particles.Sprite_Pos:NiagaraPosition=&AddVectorToPosition(Position=$Output.SkeletalMeshLocation.SampledPositionWithoutOffset, Vector=<0 0 0>)
   ) {isSetParameters=true, module=SetVariables_#3}
   &ParticleState(Kill Particles When Lifetime Has Elapsed=false, Let Infinitely Lived Particles Die When Emitter Deactivates=true)
   &ScaleMeshSize(Scale Factor=&SelectVectorFromArray(
     Direct Array Index=$Particles.PartitionParticles.Partition
     Vector Selection Array=DataInterface() {propertyValues {FloatData [<0.8 1 1>, <0.8 1 1>, <0.5 0.5 0.5>, <0.5 0.5 0.5>]}}
    )) {module=ScaleMeshSize001}
   &ScaleMeshSize(Scale Factor=&MakeVector(
     X=&Lerp_Float(
      A=&FN_FrontEndFloatModifier(Gameplay=1, Front End=0.7)
      B=&FN_FrontEndFloatModifier(Gameplay=1.4, Front End=0.8)
      Alpha=&ClampFloat(Float=$System.FN_Contrail_Setup.Dive Value)
     )
     Y=1
     Z=1
    ))
   &DynamicMaterialParameters(
    Index 0 Param 1=&FloatFromCurve(
     FloatCurve=DataInterface() {propertyValues {Curve {keys [<at 0 1>,<at 1 0>]}}}
     CurveIndex=&Divide_Float(A=&VectorLength(Vector=$Engine.Owner.Velocity), B=9000)
    )
    Index 0 Param 2=&MakeCustomFloatFromBool(Bool=&SetBoolByFloatComparison(A=$System.FN_Contrail_Setup.Dive Value, B=0.2), True Float=1.4, False Float=1)
    Index 0 Param 3=&SelectFloatFromArray(
     Direct Array Index=$Particles.PartitionParticles.Partition
     Float Selection Array=DataInterface() {propertyValues {FloatData [0,0.2,0.4,0.6]}}
    )
    Index 0 Param 4=&SelectFloatFromArray(
     Direct Array Index=$Particles.PartitionParticles.Partition
     Float Selection Array=DataInterface() {propertyValues {FloatData [0.05,0.25,0.5,0.75]}}
    )
    Index 1 Param 1=&FN_FrontEndFloatModifier(
     Gameplay=&Lerp_Float(
      A=&SelectFloatFromArray(
       Direct Array Index=$Particles.PartitionParticles.Partition
       Float Selection Array=DataInterface() {propertyValues {FloatData [0.75,0.75,0.3,0.3]}}
      )
      B=&SelectFloatFromArray(
       Direct Array Index=$Particles.PartitionParticles.Partition
       Float Selection Array=DataInterface() {propertyValues {FloatData [0.6,0.6,0.2,0.2]}}
      )
      Alpha=&ClampFloat(Float=$System.FN_Contrail_Setup.Dive Value)
     )
     Front End=0.72
    )
    Module.Param0WriteEnabled=true
    Module.Param0WriteEnabled1=true
    Module.Param1WriteEnabled=true
    Module.Param1WriteEnabled1=true
    Module.Param2WriteEnabled=true
    Module.Param2WriteEnabled1=true
    Module.Param3WriteEnabled=true
    Module.Param3WriteEnabled1=true
    Write Parameter Index 1=true
   )
   /SetVariables(
    Particles.Flame_Color:LinearColor=&Multiply_LinearColorByFloat(Linear Color=<rgba 1 1 1 1>, Float=&FN_FrontEndFloatModifier(Gameplay=1.2, Front End=1))
    Particles.Pink_DynamParam1:Vector4f=&MakeVector4(
     X=&FloatFromCurve(
      FloatCurve=DataInterface() {propertyValues {Curve {keys [<at 0 1>,<at 1 0>]}}}
      CurveIndex=&Divide_Float(A=&VectorLength(Vector=$Engine.Owner.Velocity), B=9000)
     )
     Y=&MakeCustomFloatFromBool(Bool=&SetBoolByFloatComparison(A=$System.FN_Contrail_Setup.Dive Value, B=0.2), True Float=1.4, False Float=1)
     Z=&SelectFloatFromArray(
      Direct Array Index=$Particles.PartitionParticles.Partition
      Float Selection Array=DataInterface() {propertyValues {FloatData [0,0.2,0.4,0.6]}}
     )
     W=0
    )
    Particles.Pink_DynamParam2:Vector4f=&MakeVector4(
     X=&FN_FrontEndFloatModifier(
      Gameplay=&Lerp_Float(
       A=&SelectFloatFromArray(
        Direct Array Index=$Particles.PartitionParticles.Partition
        Float Selection Array=DataInterface() {propertyValues {FloatData [0.8,0.8,0.35,0.35]}}
       )
       B=&SelectFloatFromArray(
        Direct Array Index=$Particles.PartitionParticles.Partition
        Float Selection Array=DataInterface() {propertyValues {FloatData [0.7,0.7,0.3,0.3]}}
       )
       Alpha=&ClampFloat(Float=$System.FN_Contrail_Setup.Dive Value)
      )
      Front End=0.7
     )
     Y=0
     Z=0
     W=0
    )
   ) {isSetParameters=true, module=SetVariables_#4}
   &ScaleColor(
    Scale Alpha=&Waveform(Global Amplitude Scale=2, "Waveform [1]"=::NewEnumerator9, "[1] Frequency"=5, "[1] Amplitude Min/Max"=<0 1>)
    Scale RGB=&Multiply_VectorByFloat(Vector=<1 1 1>, Float=&FN_FrontEndFloatModifier(Gameplay=1.2, Front End=1))
    ScaleA=true
    ScaleRGB=true
   )
  ]
  rendererClasses [&NiagaraMeshRendererProperties, &NiagaraSpriteRendererProperties]
  renderers [
   {rendererClass=&NiagaraMeshRendererProperties, rendererIndex=0}
   {rendererClass=&NiagaraMeshRendererProperties, rendererIndex=1}
   {rendererClass=&NiagaraSpriteRendererProperties, rendererIndex=2}
  ]
 }
}
summary {
 systemName=NS_Contrail_Medicrow
 emitters [
  {emitterName=Trail_Purple, rendererClasses [&NiagaraMeshRendererProperties, &NiagaraSpriteRendererProperties]}
  {emitterName=FlameSprites, rendererClasses [&NiagaraSpriteRendererProperties]}
  {emitterName=FlameSprites001, rendererClasses [&NiagaraSpriteRendererProperties]}
  {
   bEnabled=false
   emitterName=Frontend_Flames
   rendererClasses [&NiagaraMeshRendererProperties, &NiagaraSpriteRendererProperties]
  }
 ]
 userVariables [
  User.bIsFrontEnd {type=NiagaraBool}
  User.bIsFrontEndPreview {type=NiagaraBool}
  User.ParaGlideLeanAlpha {type=NiagaraFloat, defaultValue=0, description=""}
  User.RibbonIDOffset {type=NiagaraInt32}
 ]
}
